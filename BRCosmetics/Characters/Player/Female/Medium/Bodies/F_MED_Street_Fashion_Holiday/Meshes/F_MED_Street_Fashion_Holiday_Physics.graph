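# Physics asset for the F_MED_Street_Fashion_Holiday female player body cosmetic, covering dynamic jacket panels, a snowball charm and a can.
# Thin box bodies form seven jacket panels, each limited to its MSTR bone, with the two front panels chained in two segments.
# The snowball sphere hangs from the pelvis through two chain links, while the can bodies have no constraints.
# Capsules on pelvis, spine_03, right thigh and right forearm provide partial body shapes.
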PhysicsAsset /BRCosmetics/Characters/Player/Female/Medium/Bodies/F_MED_Street_Fashion_Holiday/Meshes/F_MED_Street_Fashion_Holiday_Physics
bodies {
 Dyn_Snowall [{shapeType=Sphere, radius=4.301, center=<4.0429 0 0>}]
 Dyn_can_MSTR [{shapeType=Capsule, length=0.505, radius=0.505}]
 dyn_can [{shapeType=Capsule, length=5.6803, radius=3.5697, center=<3.3264 -0.0571 0.502>}]
 dyn_jacket_bk_01 [{shapeType=Box, extentX=11, extentY=9.5, extentZ=0.6903, center=<5 0 0>}]
 dyn_jacket_bk_MSTR [{shapeType=Box, extentX=1.01, extentY=1.01, extentZ=1.01, rotation=<rot 0 -90 0>}]
 dyn_jacket_bk_l_01 [{shapeType=Box, extentX=11, extentY=9.5, extentZ=0.6903, center=<5 0 0>, rotation=<rot 0 0 34.9997>}]
 dyn_jacket_bk_l_MSTR [{shapeType=Box, extentX=1.01, extentY=1.01, extentZ=1.01, rotation=<rot 0 -90 0>}]
 dyn_jacket_bk_r_01 [{shapeType=Box, extentX=11, extentY=9.5, extentZ=0.6903, center=<5 0 0.0001>, rotation=<rot 0 0 -34.9996>}]
 dyn_jacket_bk_r_MSTR [{shapeType=Box, extentX=1.01, extentY=1.01, extentZ=1.01, rotation=<rot 0 -90 0>}]
 dyn_jacket_fr_l_01 [{shapeType=Box, extentX=11, extentY=9.5, extentZ=0.6903, center=<5 0 0>, rotation=<rot 0 0 -35.0004>}]
 dyn_jacket_fr_l_02 [{shapeType=Box, extentX=5.8562, extentY=9.5, extentZ=0.6903, center=<2.0136 0 0>, rotation=<rot 0 0 -35.0002>}]
 dyn_jacket_fr_l_MSTR [{shapeType=Box, extentX=1.01, extentY=1.01, extentZ=1.01, rotation=<rot 0 -90 0>}]
 dyn_jacket_fr_r_01 [{shapeType=Box, extentX=11, extentY=9.5, extentZ=1.1903, center=<5 0 0>, rotation=<rot 0 0 34.9997>}]
 dyn_jacket_fr_r_02 [{shapeType=Box, extentX=4.9335, extentY=9.5, extentZ=1.1903, center=<1.7355 0 0>, rotation=<rot 0 0 34.9995>}]
 dyn_jacket_fr_r_MSTR [{shapeType=Box, extentX=1.01, extentY=1.01, extentZ=1.01, rotation=<rot 0 -90 0>}]
 dyn_jacket_l_01 [{shapeType=Box, extentX=11, extentY=9.5, extentZ=0.6903, center=<5 0 0>, rotation=<rot 0 0 35>}]
 dyn_jacket_l_MSTR [{shapeType=Box, extentX=1.01, extentY=1.01, extentZ=1.01, rotation=<rot 0 -90 0>}]
 dyn_jacket_r_01 [{shapeType=Box, extentX=11, extentY=9.5, extentZ=1.1903, center=<5 0 0.1827>, rotation=<rot 0 0 -34.9998>}]
 dyn_jacket_r_MSTR [{shapeType=Box, extentX=1.01, extentY=1.01, extentZ=1.01, rotation=<rot 0 -90 0>}]
 dyn_snowball_chain_01 [{shapeType=Box, extentX=2.1456, extentY=5.4088, extentZ=1.01, center=<2.2507 0 0>, rotation=<rot 0 -90 0>}]
 dyn_snowball_chain_02 [{shapeType=Box, extentX=2.1932, extentY=1.01, extentZ=5.9927, center=<2.4156 0 0>, rotation=<rot 0 90 -90>}]
 lowerarm_twist_02_r [{shapeType=Capsule, length=12.134, radius=5.4484, center=<2.0274 -0.0161 -0.403>, rotation=<rot -90 0.0197 -0.0197>}]
 pelvis [
  {shapeType=Capsule, length=12.2003, radius=9.8121, center=<-4.3185 -2.1404 -0.1649>}
  {shapeType=Capsule, length=9.7075, radius=9.2201, center=<0 -1.4445 0>}
 ]
 spine_03 [{shapeType=Capsule, length=3.2426, radius=8.7172, center=<-2.0188 2.4863 0>}]
 thigh_r [{shapeType=Capsule, length=17.0367, radius=7.3722, center=<11.5563 0.0001 0>, rotation=<rot -87.3943 0 -2.5926>}]
}
constraints [
 {bone1Name=dyn_jacket_l_01, bone2Name=dyn_jacket_l_MSTR, swing1LimitDegrees=35, swing2LimitDegrees=35, twistLimitDegrees=5}
 {bone1Name=dyn_jacket_bk_l_01, bone2Name=dyn_jacket_bk_l_MSTR, swing1LimitDegrees=20, swing2LimitDegrees=30, twistLimitDegrees=5}
 {bone1Name=dyn_jacket_bk_01, bone2Name=dyn_jacket_bk_MSTR, swing1LimitDegrees=20, swing2LimitDegrees=30, twistLimitDegrees=5}
 {bone1Name=dyn_jacket_bk_r_01, bone2Name=dyn_jacket_bk_r_MSTR, swing1LimitDegrees=20, swing2LimitDegrees=30, twistLimitDegrees=5}
 {bone1Name=dyn_jacket_r_01, bone2Name=dyn_jacket_r_MSTR, swing1LimitDegrees=35, swing2LimitDegrees=35, twistLimitDegrees=5}
 {bone1Name=dyn_jacket_fr_l_01, bone2Name=dyn_jacket_fr_l_MSTR, swing1LimitDegrees=20, swing2LimitDegrees=30, twistLimitDegrees=5}
 {bone1Name=dyn_jacket_fr_l_02, bone2Name=dyn_jacket_fr_l_01, swing1LimitDegrees=20, swing2LimitDegrees=30, twistLimitDegrees=5}
 {bone1Name=dyn_jacket_fr_r_01, bone2Name=dyn_jacket_fr_r_MSTR, swing1LimitDegrees=20, swing2LimitDegrees=30, twistLimitDegrees=5}
 {bone1Name=dyn_jacket_fr_r_02, bone2Name=dyn_jacket_fr_r_01, swing1LimitDegrees=20, swing2LimitDegrees=30, twistLimitDegrees=5}
 {bone1Name=Dyn_Snowall, bone2Name=dyn_snowball_chain_02, swing1LimitDegrees=15, swing2LimitDegrees=15, twistLimitDegrees=3}
 {bone1Name=dyn_snowball_chain_02, bone2Name=dyn_snowball_chain_01, swing1LimitDegrees=35, swing2LimitDegrees=35, twistLimitDegrees=3}
 {bone1Name=dyn_snowball_chain_01, bone2Name=pelvis, swing1LimitDegrees=35, swing2LimitDegrees=35, twistLimitDegrees=3}
]
