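# Physics asset for the F_MED_Street_Fashion_Holiday female body cosmetic, set up for the motion of its hair.
# Seven two-bone chains (center back, plus front, side and back on each side) each hang from a small MSTR box body through angle-limited constraints with tight twist.
# Hair bodies are boxes, except capsule tips on the side chains.
# Unconstrained capsules on the head and upper spine give the hair shapes to collide against.
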PhysicsAsset /BRCosmetics/Characters/Player/Female/Medium/Bodies/F_MED_Street_Fashion_Holiday/Meshes/Parts/F_MED_Street_Fashion_Holiday_Physics
bodies {
 dyn_hair_bk_01 [{shapeType=Box, extentX=4.2321, extentY=11.4778, extentZ=2.2504, center=<3.1137 0 0>}]
 dyn_hair_bk_02 [{shapeType=Box, extentX=4.2157, extentY=10.0427, extentZ=3.3911, center=<1.158 0 0>}]
 dyn_hair_bk_MSTR [{shapeType=Box, extentX=1.01, extentY=1.01, extentZ=1.01}]
 dyn_hair_l_01 [{shapeType=Box, extentX=5.521, extentY=1.2147, extentZ=3.111, center=<3.5 0 0>}]
 dyn_hair_l_02 [{shapeType=Box, extentX=5, extentY=1.2935, extentZ=2.7755, center=<4 0 0>}]
 dyn_hair_l_03 [{shapeType=Box, extentX=5, extentY=6.6337, extentZ=4.7845, center=<3 0 0>, rotation=<rot 0 0 -10>}]
 dyn_hair_l_04 [{shapeType=Box, extentX=3.6923, extentY=5.1932, extentZ=7.7498, center=<1.9725 0 0>, rotation=<rot -4.6977 1.7141 -20.0729>}]
 dyn_hair_l_05 [{shapeType=Box, extentX=5, extentY=3.6596, extentZ=4, center=<3 0 0>, rotation=<rot 0 0 -10.0006>}]
 dyn_hair_l_06 [{shapeType=Capsule, length=6.1, radius=1.8493, center=<2.9354 0.5365 0.9406>, rotation=<rot 0 96.2874 -83.5111>}]
 dyn_hair_l_MSTR [{shapeType=Box, extentX=1.01, extentY=1.01, extentZ=1.01}]
 dyn_hair_l_bk_MSTR [{shapeType=Box, extentX=1.01, extentY=1.01, extentZ=1.01}]
 dyn_hair_l_fr_MSTR [{shapeType=Box, extentX=1.01, extentY=1.01, extentZ=1.01}]
 dyn_hair_r_01 [{shapeType=Box, extentX=5.521, extentY=1.2147, extentZ=3.111, center=<3.5 0 0>}]
 dyn_hair_r_02 [{shapeType=Box, extentX=5, extentY=1.2935, extentZ=2.7755, center=<4 0 0>}]
 dyn_hair_r_03 [{shapeType=Box, extentX=5, extentY=6.6337, extentZ=4.7845, center=<3 0 0>, rotation=<rot 0 0 10>}]
 dyn_hair_r_04 [{shapeType=Box, extentX=3.6923, extentY=5.1932, extentZ=7.7498, center=<1.9725 0 0>, rotation=<rot -4.6977 -1.714 20.071>}]
 dyn_hair_r_05 [{shapeType=Box, extentX=5, extentY=3.6596, extentZ=4, center=<3 0 0>, rotation=<rot 0 0 10.0003>}]
 dyn_hair_r_06 [{shapeType=Capsule, length=6.0908, radius=1.8068, center=<3.0364 -0.5878 0.7775>, rotation=<rot 0 84.5387 -87.5768>}]
 dyn_hair_r_MSTR [{shapeType=Box, extentX=1.01, extentY=1.01, extentZ=1.01}]
 dyn_hair_r_bk_MSTR [{shapeType=Box, extentX=1.01, extentY=1.01, extentZ=1.01}]
 dyn_hair_r_fr_MSTR [{shapeType=Box, extentX=1.01, extentY=1.01, extentZ=1.01}]
 head [{shapeType=Capsule, length=8.6563, radius=5.5553, center=<-0.1609 1.4323 0>, rotation=<rot -0.0001 90.0001 -89.9999>}]
 spine_05 [
  {shapeType=Capsule, length=26.623, radius=6.2489, center=<9.2825 -4.3847 0>}
  {shapeType=Capsule, length=22.6908, radius=4.9871, center=<8.6349 0.5529 0>}
 ]
}
constraints [
 {bone1Name=dyn_hair_r_05, bone2Name=dyn_hair_r_MSTR, swing1LimitDegrees=20, swing2LimitDegrees=30, twistLimitDegrees=3}
 {bone1Name=dyn_hair_r_06, bone2Name=dyn_hair_r_05, swing1LimitDegrees=15, swing2LimitDegrees=10, twistLimitDegrees=3}
 {bone1Name=dyn_hair_l_05, bone2Name=dyn_hair_l_MSTR, swing1LimitDegrees=20, swing2LimitDegrees=30, twistLimitDegrees=3}
 {bone1Name=dyn_hair_l_06, bone2Name=dyn_hair_l_05, swing1LimitDegrees=15, swing2LimitDegrees=10, twistLimitDegrees=3}
 {bone1Name=dyn_hair_l_03, bone2Name=dyn_hair_l_bk_MSTR, swing1LimitDegrees=20, swing2LimitDegrees=30, twistLimitDegrees=3}
 {bone1Name=dyn_hair_l_04, bone2Name=dyn_hair_l_03, swing1LimitDegrees=20, swing2LimitDegrees=30, twistLimitDegrees=3}
 {bone1Name=dyn_hair_r_03, bone2Name=dyn_hair_r_bk_MSTR, swing1LimitDegrees=20, swing2LimitDegrees=30, twistLimitDegrees=3}
 {bone1Name=dyn_hair_r_04, bone2Name=dyn_hair_r_03, swing1LimitDegrees=20, swing2LimitDegrees=30, twistLimitDegrees=3}
 {bone1Name=dyn_hair_bk_01, bone2Name=dyn_hair_bk_MSTR, swing1LimitDegrees=20, swing2LimitDegrees=30, twistLimitDegrees=3}
 {bone1Name=dyn_hair_bk_02, bone2Name=dyn_hair_bk_01, swing1LimitDegrees=20, swing2LimitDegrees=30, twistLimitDegrees=3}
 {bone1Name=dyn_hair_r_01, bone2Name=dyn_hair_r_fr_MSTR, swing1LimitDegrees=20, swing2LimitDegrees=30, twistLimitDegrees=3}
 {bone1Name=dyn_hair_r_02, bone2Name=dyn_hair_r_01, swing1LimitDegrees=15, swing2LimitDegrees=10, twistLimitDegrees=3}
 {bone1Name=dyn_hair_l_02, bone2Name=dyn_hair_l_01, swing1LimitDegrees=15, swing2LimitDegrees=10, twistLimitDegrees=3}
 {bone1Name=dyn_hair_l_01, bone2Name=dyn_hair_l_fr_MSTR, swing1LimitDegrees=20, swing2LimitDegrees=30, twistLimitDegrees=3}
]
