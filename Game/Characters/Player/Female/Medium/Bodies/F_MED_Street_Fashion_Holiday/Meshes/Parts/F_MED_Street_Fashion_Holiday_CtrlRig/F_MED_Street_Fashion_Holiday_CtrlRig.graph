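# Control Rig for the F_MED_Street_Fashion_Holiday player character part that procedurally bends back and rear-side hair bones as the head rotates.
# It reads the head's transform relative to spine_05 and takes the Z Euler angle as the single driver.
# Clamped remaps turn that angle into additive local rotations, applied to the chain base bones, then their child bones.
# The child side bones receive mirrored Z rotations.
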ControlRigBlueprint /Game/Characters/Player/Female/Medium/Bodies/F_MED_Street_Fashion_Holiday/Meshes/Parts/F_MED_Street_Fashion_Holiday_CtrlRig/F_MED_Street_Fashion_Holiday_CtrlRig
graphs [
 RigVMFunctionLibrary {links [], nodes []}
 RigVMModel {nodes [
   RigUnit_BeginExecution = &RigUnit_BeginExecution()
   n1 = &RigUnit_GetRelativeBoneTransform(Bone=head, Space=spine_05)
   n2 = &RigVMFunction_MathQuaternionToEuler(RotationOrder=ZYX, Value=@n1.Transform.Rotation)
   n3 = &RigVMFunction_MathQuaternionToEuler(RotationOrder=ZYX, Value=@n1.Transform.Rotation)
   n4 = &RigVMFunction_MathQuaternionFromEuler(
    Euler=(X=0,Y=0,Z=0)
    Euler.Y=&RigVMFunction_MathFloatRemap(SourceMaximum=-40, SourceMinimum=-4, TargetMaximum=25, bClamp=true, Value=@n3.Result.Z).Result
   )
   /Script/ControlRig.RigUnit_ModifyBoneTransforms(
    BoneToModify="((Bone=\"dyn_hair_bk_01\",Transform=(Rotation=(X=0,Y=0,Z=0,W=1),Translation=(X=0,Y=0,Z=0),Scale3D=(X=1,Y=1,Z=1))),(Bone=\"dyn_hair_l_03\",Transform=(Rotation=(X=0,Y=0,Z=0,W=1),Translation=(X=0,Y=0,Z=0),Scale3D=(X=1,Y=1,Z=1))),(Bone=\"dyn_hair_r_03\",Transform=(R ...(truncated)"
    Mode=AdditiveLocal
    Weight=1
    WeightMaximum=1
    BoneToModify.0.Transform.Rotation=&RigVMFunction_MathQuaternionFromEuler(
     Euler=(X=0,Y=0,Z=0)
     Euler.Y=&RigVMFunction_MathFloatRemap(SourceMaximum=-40, SourceMinimum=-4, TargetMaximum=45, bClamp=true, Value=@n2.Result.Z).Result
    ).Result
    BoneToModify.1.Transform.Rotation=@n4.Result
    BoneToModify.2.Transform.Rotation=@n4.Result
    ExecutePin=@RigUnit_BeginExecution.ExecutePin
   ) {is=struct RigVMUnitNode}
   > &RigUnit_ModifyBoneTransforms(
    BoneToModify="((Bone=\"dyn_hair_bk_02\",Transform=(Rotation=(X=0,Y=0,Z=0,W=1),Translation=(X=0,Y=0,Z=0),Scale3D=(X=1,Y=1,Z=1))),(Bone=\"dyn_hair_l_04\",Transform=(Rotation=(X=0,Y=0,Z=0,W=1),Translation=(X=0,Y=0,Z=0),Scale3D=(X=1,Y=1,Z=1))),(Bone=\"dyn_hair_r_04\",Transform=(R ...(truncated)"
    Mode=AdditiveLocal
    Weight=1
    WeightMaximum=1
    BoneToModify.1.Transform.Rotation=&RigVMFunction_MathQuaternionFromEuler(
     Euler=(X=0,Y=0,Z=0)
     Euler.Z=&RigVMFunction_MathFloatRemap(SourceMaximum=-40, SourceMinimum=-4, TargetMaximum=25, bClamp=true, Value=@n3.Result.Z).Result
    ).Result
    BoneToModify.2.Transform.Rotation=&RigVMFunction_MathQuaternionFromEuler(
     Euler=(X=0,Y=0,Z=0)
     Euler.Z=&RigVMFunction_MathFloatRemap(SourceMaximum=-40, SourceMinimum=-4, TargetMaximum=-25, bClamp=true, Value=@n3.Result.Z).Result
    ).Result
    BoneToModify.0.Transform.Rotation=&RigVMFunction_MathQuaternionFromEuler(
     Euler=(X=0,Y=0,Z=0)
     Euler.Y=&RigVMFunction_MathFloatRemap(SourceMaximum=-40, SourceMinimum=-4, TargetMaximum=-50, bClamp=true, Value=@n3.Result.Z).Result
    ).Result
   )
  ]}
]
hierarchy [
 root {loc=<0 0 0>, rot=<0 0 0 1>}
 pelvis {parent=root, loc=<0 -0.5301 98.0571>, rot=<0.0127 0.707 -0.0127 -0.707>}
 spine_01 {parent=pelvis, loc=<6.1208 -0.9328 0>, rot=<0 0 0.034 -0.9994>}
 spine_02 {parent=spine_01, loc=<6.5996 -0.2095 0>, rot=<0 0 0.0238 -0.9997>}
 spine_03 {parent=spine_02, loc=<6.4914 -0.1237 0>, rot=<0 0 0.0149 -0.9999>}
 spine_04 {parent=spine_03, loc=<6.9598 0.0794 0>, rot=<0 0 -0.0132 -0.9999>}
 spine_05 {parent=spine_04, loc=<6.9987 0 0>, rot=<0 0 -0.0015 -1>}
 clavicle_l {parent=spine_05, loc=<5.4831 -0.1053 -3.2946>, rot=<0.0465 -0.7289 -0.009 -0.6829>}
 clavicle_r {parent=spine_05, loc=<5.4831 -0.1053 3.2946>, rot=<0.7289 0.0465 -0.6829 0.009>}
 neck_01 {parent=spine_05, loc=<11.3138 0.1529 0>, rot=<0 0 -0.2003 -0.9797>}
 neck_02 {parent=neck_01, loc=<4.5627 0 0>, rot=<0 0 0.0504 -0.9987>}
 head {parent=neck_02, loc=<5.3914 0 0>, rot=<0 0 0.115 -0.9934>}
 faceAttach {parent=head, loc=<-4.8249 -0.2336 -0.0373>, rot=<0.0082 -0.7071 -0.0082 -0.7071>, scale=<0.89 0.89 0.89>}
 C_brow_mid {parent=faceAttach, loc=<-0.0755 9.9174 10.2264>, rot=<0 0.017 -0.0034 -0.9998>}
 L_brow_mid {parent=faceAttach, loc=<3.2495 9.221 11.0756>, rot=<-0.0001 -0.0043 0.2553 -0.9669>}
 L_brow_outer {parent=faceAttach, loc=<5.8639 6.8829 11.3037>, rot=<0.1218 0.1494 0.4055 -0.8936>}
 L_cheek_inner {parent=faceAttach, loc=<4.7852 7.5354 4.1283>, rot=<-0.0262 -0.157 0.3316 -0.9299>}
 C_lip_upper_mid {parent=faceAttach, loc=<-0.0397 9.5719 2.3275>, rot=<0.2225 0.0024 -0.0009 -0.9749>}
 L_lip_upper_outer {parent=faceAttach, loc=<1.8453 9.0294 2.4343>, rot=<0.0417 -0.0566 0.2566 -0.964>}
 L_lip_corner {parent=faceAttach, loc=<2.3815 8.5891 2.3505>, rot=<0.0027 0.0083 -0.0146 -0.9999>}
 C_nose_bridge {parent=faceAttach, loc=<-0.0287 10.9394 7.0288>, rot=<-0.2217 -0.0326 -0.0092 -0.9745>}
 R_brow_mid {parent=faceAttach, loc=<-3.3316 9.2252 11.0801>, rot=<-0.0001 0.0043 -0.2553 -0.9669>}
 R_brow_outer {parent=faceAttach, loc=<-5.956 6.8644 11.2891>, rot=<0.0662 -0.0795 -0.4247 -0.8994>}
 R_cheek_inner {parent=faceAttach, loc=<-4.8569 7.5455 4.1196>, rot=<-0.0266 0.1581 -0.3315 -0.9297>}
 R_lip_corner {parent=faceAttach, loc=<-2.4849 8.6173 2.3595>, rot=<0.0028 -0.0081 0.0146 -0.9999>}
 C_jaw {parent=faceAttach, loc=<-0.1119 1.508 5.4586>, rot=<0 0 0 1>}
 L_lip_lower_outer {parent=C_jaw, loc=<1.9851 7.4432 -3.0568>, rot=<0.0109 0.0332 0.2971 -0.9542>}
 C_lip_lower_mid {parent=C_jaw, loc=<0.1258 1.2608 -3.1837>, rot=<0 -0.0226 -0.0048 -0.9997>}
 R_lip_lower_outer {parent=C_jaw, loc=<-1.5846 7.3914 -3.1004>, rot=<0.001 0.0202 -0.2558 -0.9665>}
 R_eye {parent=faceAttach, loc=<-3.2665 6.4996 9.5052>, rot=<0 0 0.9998 -0.0187>}
 R_eye_lid_upper_mid {parent=R_eye, loc=<-0.4046 -2.1966 0.6072>, rot=<0.0523 -0.0009 0.0167 -0.9985>}
 R_eye_lid_lower_mid {parent=R_eye, loc=<-0.3825 -1.8684 -0.8835>, rot=<0.0329 -0.0005 0.0167 -0.9993>}
 L_eye {parent=faceAttach, loc=<3.1826 6.4996 9.5052>, rot=<0 0 0.9998 0.0187>}
 L_eye_lid_upper_mid {parent=L_eye, loc=<0.2583 -2.2489 0.5974>, rot=<0.0523 0 0 -0.9986>}
 L_eye_lid_lower_mid {parent=L_eye, loc=<0.3978 -1.8483 -0.9283>, rot=<0.0329 0 0 -0.9995>}
 R_lip_upper_outer {parent=faceAttach, loc=<-1.94 9.0891 2.4729>, rot=<0.0464 0.0432 -0.2514 -0.9658>}
 dyn_hair_r_MSTR {parent=head, loc=<2.5384 0.2339 6.5697>, rot=<0 0 0 -1>}
 dyn_hair_r_05 {parent=dyn_hair_r_MSTR, loc=<-0.0012 0.1512 0.7857>, rot=<-0.0022 -0.771 -0.6368 -0.0039>}
 dyn_hair_r_06 {parent=dyn_hair_r_05, loc=<6.8107 0.5199 0>, rot=<0.0379 0.2371 -0.0099 -0.9707>}
 dyn_hair_l_MSTR {parent=head, loc=<2.5384 0.2339 -6.5697>, rot=<0 0 0 -1>}
 dyn_hair_l_05 {parent=dyn_hair_l_MSTR, loc=<-0.0012 0.1494 -0.7764>, rot=<0.0039 0.6368 0.771 0.0022>}
 dyn_hair_l_06 {parent=dyn_hair_l_05, loc=<6.8107 -0.4437 0>, rot=<-0.0335 0.2377 0.0088 -0.9707>}
 dyn_hair_l_bk_MSTR {parent=head, loc=<1.9388 -2.5776 -6.9032>, rot=<0 0 0 -1>}
 dyn_hair_l_03 {parent=dyn_hair_l_bk_MSTR, loc=<0 0 0>, rot=<0.0913 -0.9698 0.2108 -0.0815>}
 dyn_hair_l_04 {parent=dyn_hair_l_03, loc=<4.2224 0 0>, rot=<-0.6187 -0.0143 0.1381 -0.7733>}
 dyn_hair_r_bk_MSTR {parent=head, loc=<1.9388 -2.5796 6.8482>, rot=<0 0 0 -1>}
 dyn_hair_r_03 {parent=dyn_hair_r_bk_MSTR, loc=<0 0 0>, rot=<0.0856 -0.1876 0.9742 -0.0916>}
 dyn_hair_r_04 {parent=dyn_hair_r_03, loc=<4.229 0 0>, rot=<0.636 -0.0164 -0.1441 -0.7579>}
 dyn_hair_bk_MSTR {parent=head, loc=<1.8158 -5.9288 -0.038>, rot=<0 0 0 -1>}
 dyn_hair_bk_01 {parent=dyn_hair_bk_MSTR, loc=<0 0 0>, rot=<0.1067 -0.699 0.699 -0.1067>}
 dyn_hair_bk_02 {parent=dyn_hair_bk_01, loc=<4.102 0 0>, rot=<0 -0.1468 0 -0.9892>}
 dyn_hair_r_fr_MSTR {parent=head, loc=<1.5892 2.0751 6.2139>, rot=<0 0 0 -1>}
 dyn_hair_r_01 {parent=dyn_hair_r_fr_MSTR, loc=<0 0 0>, rot=<-0.0399 -0.7694 -0.6356 -0.0495>}
 dyn_hair_r_02 {parent=dyn_hair_r_01, loc=<5.7922 0 0>, rot=<0.0372 0.1794 -0.0122 -0.983>}
 dyn_hair_l_fr_MSTR {parent=head, loc=<1.5892 2.0751 -6.2139>, rot=<0 0 0 -1>}
 dyn_hair_l_01 {parent=dyn_hair_l_fr_MSTR, loc=<0 0 0>, rot=<0.0495 0.6356 0.7694 0.0399>}
 dyn_hair_l_02 {parent=dyn_hair_l_01, loc=<5.7922 0 0>, rot=<-0.0329 0.1799 0.0108 -0.9831>}
]
