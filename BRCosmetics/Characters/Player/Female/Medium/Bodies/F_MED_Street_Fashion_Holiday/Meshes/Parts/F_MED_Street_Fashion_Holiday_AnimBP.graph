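# Animation Blueprint for the parts mesh of the F_MED_Street_Fashion_Holiday player cosmetic, adding secondary physics to the pose copied from its attached parent mesh.
# Two identically tuned RigidBody branches simulate relative to spine_05 with damped component motion and overridden gravity.
# A short bool blend on bIsTargeting outputs physics alone while targeting, and otherwise physics followed by the outfit's Control Rig.
# The event graph's update event is empty.
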AnimBlueprint /BRCosmetics/Characters/Player/Female/Medium/Bodies/F_MED_Street_Fashion_Holiday/Meshes/Parts/F_MED_Street_Fashion_Holiday_AnimBP
target_skeleton=/Game/Characters/Player/Female/Medium/Bodies/F_MED_Street_Fashion_Holiday/Meshes/Parts/F_MED_Street_Fashion_Holiday_Skeleton
graphs [
 AnimGraph {nodes [
   n1 = RigidBody(ComponentPose=Local To Component(LocalPose=Copy Pose From Mesh() {bUseAttachedParent=true})) {
    bClampLinearTranslationLimitToRefPose=true
    bOverrideWorldGravity=true
    overridePhysicsAsset=&/F_MED_Street_Fashion_Holiday_Physics
    simulationSpace=BaseBoneSpace
    alphaBoolBlend {blendOption=Linear}
    baseBoneRef {boneName=spine_05}
    componentAppliedLinearAccClamp=<1000 1000 1000>
    componentLinearAccScale=<0.3 0.3 0.5>
    componentLinearVelScale=<0.2 0.2 0.3>
    overrideWorldGravity {z=-980}
   }
   n2 = RigidBody(ComponentPose=Local To Component(LocalPose=Copy Pose From Mesh() {bUseAttachedParent=true})) {
    bClampLinearTranslationLimitToRefPose=true
    bOverrideWorldGravity=true
    overridePhysicsAsset=&F_MED_Street_Fashion_Holiday_Physics
    simulationSpace=BaseBoneSpace
    alphaBoolBlend {blendOption=Linear}
    baseBoneRef {boneName=spine_05}
    componentAppliedLinearAccClamp=<1000 1000 1000>
    componentLinearAccScale=<0.3 0.3 0.5>
    componentLinearVelScale=<0.2 0.2 0.3>
    overrideWorldGravity {z=-980}
   }
   n3 = Control Rig!(Source=Component To Local(ComponentPose=@n2), bAlphaBoolEnabled:bool=True) -> (Pose) {
    alphaInputType=Bool
    alphaBoolBlend {blendOption=Linear}
    controlRigAssetReference {blueprintRigClass=/Game/Characters/Player/Female/Medium/Bodies/F_MED_Street_Fashion_Holiday/Meshes/Parts/F_MED_Street_Fashion_Holiday_CtrlRig/F_MED_Street_Fashion_Holiday_CtrlRig.F_MED_Street_Fashion_Holiday_CtrlRig_C}
   }
   Output Pose(Result=Blend Poses by bool(BlendPose_0=Component To Local(ComponentPose=@n1), BlendPose_1=@n3, BlendTime_0=0.2, BlendTime_1=0.2, bActiveValue=Get bIsTargeting()) {blendType=Linear}) {layerGroup=None}
  ]}
 EventGraph {nodes [Event BlueprintUpdateAnimation()]}
]
variables []
